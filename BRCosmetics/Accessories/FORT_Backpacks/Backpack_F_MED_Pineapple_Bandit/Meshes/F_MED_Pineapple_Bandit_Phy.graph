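# Physics asset for the Pineapple Bandit backpack cosmetic, giving its dangling zipper and money pieces secondary motion.
# A capsule pack body anchors a three-link zipper chain of capsules with limited swing and tight twist.
# A small money base capsule rotates freely on the pack, its angular limits unused.
# Eleven thin money boxes attach directly to that base in a star layout, each with asymmetric swing and tight twist limits.
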PhysicsAsset /BRCosmetics/Accessories/FORT_Backpacks/Backpack_F_MED_Pineapple_Bandit/Meshes/F_MED_Pineapple_Bandit_Phy
bodies {
 dyn_money_01 [{shapeType=Box, extentX=6, extentY=6, extentZ=1, center=<4 0 0>}]
 dyn_money_010 [{shapeType=Box, extentX=3, extentY=6, extentZ=1.3, center=<4 0 0>}]
 dyn_money_011 [{shapeType=Box, extentX=3, extentY=6, extentZ=1.3, center=<4 0 0>}]
 dyn_money_02 [{shapeType=Box, extentX=6, extentY=6, extentZ=1, center=<4 0 0.3>, rotation=<rot -5 0 0>}]
 dyn_money_03 [{shapeType=Box, extentX=3, extentY=4.5, extentZ=1, center=<2 0 0>}]
 dyn_money_04 [{shapeType=Box, extentX=2.5, extentY=4, extentZ=1, center=<3 0 0>, rotation=<rot 0 0 15.0008>}]
 dyn_money_05 [{shapeType=Box, extentX=5, extentY=6, extentZ=1, center=<3 0 0>}]
 dyn_money_06 [{shapeType=Box, extentX=5, extentY=6, extentZ=1, center=<3 0 0>}]
 dyn_money_07 [{shapeType=Box, extentX=4.5, extentY=6, extentZ=1.8, center=<3.8 0 0>}]
 dyn_money_08 [{shapeType=Box, extentX=3.7, extentY=6, extentZ=1, center=<3.6 0 0>}]
 dyn_money_09 [{shapeType=Box, extentX=4.3, extentY=6, extentZ=2.3, center=<3 0 0>}]
 dyn_money_base [{shapeType=Capsule, length=1, radius=1}]
 dyn_zipper_01 [{shapeType=Capsule, length=2, radius=1, center=<2 0 0>, rotation=<rot 90 0 0>}]
 dyn_zipper_02 [{shapeType=Capsule, length=2, radius=1, center=<2 0 0>, rotation=<rot 90 0 0>}]
 dyn_zipper_03 [{shapeType=Capsule, length=2, radius=1, center=<2 0 0>, rotation=<rot 90 0 0>}]
 pack [{shapeType=Capsule, length=8.263, radius=9.1199, center=<-1.3264 -5.8661 -11.347>}]
}
constraints [
 {bone1Name=dyn_zipper_01, bone2Name=pack, swing1LimitDegrees=60, swing2LimitDegrees=60, twistLimitDegrees=5}
 {bone1Name=dyn_zipper_02, bone2Name=dyn_zipper_01, swing1LimitDegrees=60, swing2LimitDegrees=60, twistLimitDegrees=5}
 {bone1Name=dyn_zipper_03, bone2Name=dyn_zipper_02, swing1LimitDegrees=60, swing2LimitDegrees=60, twistLimitDegrees=5}
 {
  bone1Name=dyn_money_base
  bone2Name=pack
  swing1LimitDegrees=45
  swing1Motion=Free
  swing2LimitDegrees=45
  swing2Motion=Free
  twistLimitDegrees=45
  twistMotion=Free
 }
 {bone1Name=dyn_money_01, bone2Name=dyn_money_base, swing1LimitDegrees=15, swing2LimitDegrees=35, twistLimitDegrees=5}
 {bone1Name=dyn_money_02, bone2Name=dyn_money_base, swing1LimitDegrees=15, swing2LimitDegrees=35, twistLimitDegrees=5}
 {bone1Name=dyn_money_03, bone2Name=dyn_money_base, swing1LimitDegrees=15, swing2LimitDegrees=35, twistLimitDegrees=5}
 {bone1Name=dyn_money_04, bone2Name=dyn_money_base, swing1LimitDegrees=15, swing2LimitDegrees=35, twistLimitDegrees=5}
 {bone1Name=dyn_money_05, bone2Name=dyn_money_base, swing1LimitDegrees=15, swing2LimitDegrees=35, twistLimitDegrees=5}
 {bone1Name=dyn_money_06, bone2Name=dyn_money_base, swing1LimitDegrees=15, swing2LimitDegrees=35, twistLimitDegrees=5}
 {bone1Name=dyn_money_07, bone2Name=dyn_money_base, swing1LimitDegrees=15, swing2LimitDegrees=35, twistLimitDegrees=5}
 {bone1Name=dyn_money_08, bone2Name=dyn_money_base, swing1LimitDegrees=15, swing2LimitDegrees=35, twistLimitDegrees=5}
 {bone1Name=dyn_money_09, bone2Name=dyn_money_base, swing1LimitDegrees=15, swing2LimitDegrees=35, twistLimitDegrees=5}
 {bone1Name=dyn_money_010, bone2Name=dyn_money_base, swing1LimitDegrees=15, swing2LimitDegrees=35, twistLimitDegrees=5}
 {bone1Name=dyn_money_011, bone2Name=dyn_money_base, swing1LimitDegrees=15, swing2LimitDegrees=35, twistLimitDegrees=5}
]
